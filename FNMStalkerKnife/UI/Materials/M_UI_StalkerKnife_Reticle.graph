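# Translucent UI material that draws the Stalker Knife reticle brackets, switching between cooldown and active shapes with the ReconReady parameter.
# UVs are mirrored by a symmetry function, offset and scaled about center, then sample two MSDF textures.
# Antialiased thresholds on each distance field build Bracket and wider BracketsShadow masks as named reroutes, blended by ReconReady.
# ReconReady also drives a brightness or opacity multiplier on a rerouted mask.
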Material /FNMStalkerKnife/UI/Materials/M_UI_StalkerKnife_Reticle
graph {
 properties {bAutomaticallySetUsageInEditor=false, blendMode=BLEND_Translucent, materialDomain=MD_UI, usedShadingModels=Default Lit}
 nodes [
  n1 = Constant() {r=0.5}
  n2 = ScalarParameter() {defaultValue=1, parameterName=ReconReady}
  Multiply(A=Multiply(A=NamedRerouteUsage(), B=LinearInterpolate(A=@n1, Alpha=@n2) {constA=0.2}), B=Constant() {r=0.9})
  n3 = TextureCoordinate()
  n4 = Constant() {r=0.2}
  n5 = Constant()
  n6 = MaterialFunctionCall(
   UVs=Add(
    A=MaterialFunctionCall!(UV=@n3, Mirror Y, Swizzle UV) {materialFunction=/FortUILibrary/MaterialFunctions/Basics/MF_UI_Symmetry_2x1_New}.Symmetry
    B=AppendVector(A=@n4, B=@n5)
   ) {constB=0}
   Texture Scale=Constant2Vector() {g=0.45, r=1}
  ) {materialFunction=&ScaleUVsByCenter}
  n7 = MaterialFunctionCall!(MSDF RGB=TextureSample(UVs=@n6.UVs) {samplerSource=SSM_Clamp_WorldGroupSettings, texture=/FNMStalkerKnife/UI/Textures/T_UI_MSDF_StalkerKnife_Reticle_Cooldown}.RGB) {materialFunction=/FortUILibrary/MaterialFunctions/Shapes/Modifiers/MF_UI_MSDFtoSDF}
  n8 = MaterialFunctionCall(DistanceField=@n7.SDF, Threshold=Constant() {r=0.7}) {materialFunction=&MF_UI_AntiAliasing}
  n9 = MaterialFunctionCall(DistanceField=@n7.SDF, Threshold=Constant() {r=0.3}) {materialFunction=&MF_UI_AntiAliasing}
  n10 = MaterialFunctionCall(MSDF RGB=TextureSample(UVs=@n6.UVs) {samplerSource=SSM_Clamp_WorldGroupSettings, texture=/FNMStalkerKnife/UI/Textures/T_UI_MSDF_StalkerKnife_Reticle_Active}.RGB) {materialFunction=&MF_UI_MSDFtoSDF}
  n11 = MaterialFunctionCall(DistanceField=@n10.SDF, Threshold=Constant() {r=0.7}) {materialFunction=&MF_UI_AntiAliasing}
  n12 = MaterialFunctionCall(DistanceField=@n10.SDF, Threshold=Constant() {r=0.3}) {materialFunction=&MF_UI_AntiAliasing}
  NamedRerouteDeclaration(None=LinearInterpolate(
    A=LinearInterpolate(A=@n8.Outer, B=@n9.Outer)
    B=LinearInterpolate(A=@n11.Outer, B=@n12.Outer)
    Alpha=ScalarParameter() {defaultValue=1, parameterName=ReconReady}
   )) {name=BracketsShadow, nodeColor=<rgba 1 0 0.5859 1>}
  NamedRerouteDeclaration(None=LinearInterpolate(A=@n8.Outer, B=@n11.Outer, Alpha=ScalarParameter() {defaultValue=1, parameterName=ReconReady})) {name=Bracket, nodeColor=<rgba 1 0.9609 0 1>}
  NamedRerouteUsage()
 ]
}
