WidgetBlueprint /FNMStalkerKnife/UI/Widgets/WBP_StalkerKnife_Reticle
animations [
 ReconActive {
  binding_count=1
  length_seconds=0.15
  object=&self:ReconActive
  sequence {
   bindings [Reticle_Shape {
     id=#1
     possessed_class=Image
     tracks [{
       type=MovieSceneWidgetMaterialTrack
       sections [{
         type=MovieSceneParameterSection
         end_frame=9
         start_frame=0
         channels [ReconReady {type=MovieSceneScriptingFloatChannel, keys [[0,0],[9,1]]}]
        }]
      }]
    }]
   display_rate {numerator=60}
  }
 }
 Anim_FadeIn {
  binding_count=1
  length_seconds=0.2333
  object=&self:Anim_FadeIn
  sequence {
   bindings [Reticle_ReconCooldown {
     id=#2
     possessed_class=Image
     tracks [{
       type=MovieSceneWidgetMaterialTrack
       sections [{type=MovieSceneParameterSection, end_frame=15, start_frame=1, channels [Opacity {type=MovieSceneScriptingFloatChannel, keys [[1,0],[15,1]]}]}]
      }]
    }]
   display_rate {numerator=60}
  }
 }
]
graphs [
 CheckIfIsLocalPlayer {nodes [
   n1 = CheckIfIsLocalPlayer!() -> (then, View Target:Actor Object Reference) {is=title K2Node_FunctionEntry}
   > Return Node!(
    execute
    isLocal:bool="Equal (Object)"(A=GetControllerFromActor!(Actor:Actor Object Reference=@n1) -> (ReturnValue:Controller Object Reference), B=GetOwningPlayer())
   )
  ]}
 EventGraph {nodes [
   n1 = OnViewTargetChanged!() -> (
    OutputDelegate
    then
    PlayerController:Fort Player Controller Object Reference
    OldViewTarget:Actor Object Reference
    NewViewTarget:Actor Object Reference
   ) {is=title K2Node_CustomEvent}
   > n2 = HasMatchingGameplayTag!(
    execute:Exec
    self:Object Reference=@n1.NewViewTarget
    TagToCheck:Gameplay Tag Structure=Get CooldownTag() -> (CooldownTag:Gameplay Tag Structure)
   ) -> (then:Exec, ReturnValue:bool) {is=title K2Node_Message}
   > UpdateReconReticle!(
    execute
    self
    isReconActive:bool=NOT Boolean(A=@n2)
    Animate:bool=false
    UpdateCooldownBar:bool=CheckIfIsLocalPlayer!(View Target:Actor Object Reference=@n1.NewViewTarget) -> (isLocal:bool) {is=title K2Node_CallFunction}
   ) -> (then)
   n3 = Get WaitGameplayTagAddAsyncAction() -> (WaitGameplayTagAddAsyncAction:Ability Async Wait Gameplay Tag Added Object Reference)
   > n4 = Is Valid(InputObject=@n3)
   > n5 = EndAction(self:Ability Async Object Reference=@n3)
   n6 = WaitGameplayTagAddToActor!(
    execute:Exec=[@n5.then, @n4.Is Not Valid]
    TargetActor:Actor Object Reference=@n1.NewViewTarget
    Tag:Gameplay Tag Structure=Get CooldownTag()
    OnlyTriggerOnce:bool=false
   ) -> (then:Exec, AsyncTaskProxy:Ability Async Wait Gameplay Tag Added Object Reference, Added:Exec) {is=title K2Node_AsyncAction}
   > Set WaitGameplayTagAddAsyncAction(WaitGameplayTagAddAsyncAction:Ability Async Wait Gameplay Tag Added Object Reference=@n6) -> (Output_Get:Ability Async Wait Gameplay Tag Added Object Reference)
   OnReconAbilityActivated!(execute=@n6.Added, self, ViewTarget:Actor Object Reference=@n1.NewViewTarget) -> (then)
   n7 = OnReconAbilityActivated!() -> (OutputDelegate, then, ViewTarget:Actor Object Reference) {is=title K2Node_CustomEvent}
   > n8 = Branch(Condition=CheckIfIsLocalPlayer(View Target=@n7.ViewTarget))
   > n9 = GetCooldownTimeRemainingAndDuration!(
    execute
    Actor:Actor Object Reference=@n7.ViewTarget
    CooldownTags:Gameplay Tag Container Structure=MakeGameplayTagContainerFromTag!(SingleTag:Gameplay Tag Structure=Get CooldownTag()) -> (ReturnValue:Gameplay Tag Container Structure)
   ) -> (then, TimeRemaining:float, CooldownDuration:float)
   > Set CooldownTimeRemaining(CooldownTimeRemaining:double=@n9) -> (Output_Get:double)
   > n10 = Set CooldownDuration(CooldownDuration:float=@n9.CooldownDuration) -> (Output_Get:float)
   n11 = Get CooldownTimeRemaining() -> (CooldownTimeRemaining:double)
   > SetScalarParameterValue(
    self=Get DynamicMaterial_CooldownBar() -> (DynamicMaterial_CooldownBar:Material Instance Dynamic Object Reference)
    ParameterName=Progress
    Value=float - float(A=1, B=float / float(A=@n11, B=@n10))
   )
   > n12 = Set Timer by Event(Delegate=Create Event(), Time=Get UpdateBarTick() -> (UpdateBarTick:double), bLooping=true, bMaxOncePerFrame=true)
   > Set UpdateCooldownBarTimer(UpdateCooldownBarTimer:Timer Handle Structure=@n12) -> (Output_Get:Timer Handle Structure)
   > UpdateReconReticle(isReconActive=false, Animate=true, UpdateCooldownBar=true)
   SetScalarParameterValue(
    execute=@n8.else
    self=Get Dynamic Material Reticle Shape() -> (Dynamic Material Reticle Shape:Material Instance Dynamic Object Reference)
    ParameterName=ReconReady
    Value=0
   )
   n13 = Get Spectator Tag Remove Async Action() -> (Spectator Tag Remove Async Action:Ability Async Wait Gameplay Tag Removed Object Reference)
   > n14 = Is Valid(InputObject=@n13)
   > n15 = EndAction(self=@n13)
   n16 = WaitGameplayTagRemoveFromActor!(
    execute:Exec=[@n15.then, @n14.Is Not Valid]
    TargetActor:Actor Object Reference=@n7.ViewTarget
    Tag:Gameplay Tag Structure=Get CooldownTag()
    OnlyTriggerOnce:bool=false
   ) -> (then:Exec, AsyncTaskProxy:Ability Async Wait Gameplay Tag Removed Object Reference, Removed:Exec) {is=title K2Node_AsyncAction}
   > Set Spectator Tag Remove Async Action(Spectator Tag Remove Async Action:Ability Async Wait Gameplay Tag Removed Object Reference=@n16) -> (Output_Get:Ability Async Wait Gameplay Tag Removed Object Reference)
   UpdateReconReticle(execute=@n16.Removed, isReconActive=true, Animate=true, UpdateCooldownBar=false)
   Event Construct()
   > n17 = Cast To FortPlayerController(Object=GetOwningPlayer())
   > n18 = Set As Fort Player Controller(As Fort Player Controller:Fort Player Controller Object Reference=@n17) -> (Output_Get:Fort Player Controller Object Reference)
   > Bind Event to On View Target Changed(self:Fort Player Controller Object Reference=@n18, Delegate=Create Event()) {is=title K2Node_AddDelegate}
   > OnViewTargetChanged!(
    execute
    self
    PlayerController:Fort Player Controller Object Reference
    OldViewTarget:Actor Object Reference
    NewViewTarget:Actor Object Reference=GetViewTarget!(self:Controller Object Reference=@n17) -> (ReturnValue:Actor Object Reference)
   ) -> (then) {is=title K2Node_CallFunction}
   Event Destruct()
   > Unbind Event from On View Target Changed(
    self:Fort Player Controller Object Reference=Get As Fort Player Controller() -> (As Fort Player Controller:Fort Player Controller Object Reference)
    Delegate=Create Event()
   ) {is=title K2Node_RemoveDelegate}
   > Clear and Invalidate Timer by Handle(Handle=Get UpdateCooldownBarTimer() -> (UpdateCooldownBarTimer:Timer Handle Structure))
   > EndAction(self=Get WaitGameplayTagAddAsyncAction())
   > EndAction(self=Get Spectator Tag Remove Async Action())
   Event OnInitialized()
   > n19 = GetDynamicMaterial(self=Get Reticle_ReconCooldown() -> (Reticle_ReconCooldown:Image Object Reference))
   > Set DynamicMaterial_CooldownBar(DynamicMaterial_CooldownBar:Material Instance Dynamic Object Reference=@n19) -> (Output_Get:Material Instance Dynamic Object Reference)
   > n20 = GetDynamicMaterial(self=Get Reticle_Shape() -> (Reticle_Shape:Image Object Reference))
   > Set Dynamic Material Reticle Shape(Dynamic Material Reticle Shape:Material Instance Dynamic Object Reference=@n20) -> (Output_Get:Material Instance Dynamic Object Reference)
  ]}
 UpdateCooldownProgressBar {nodes [
   UpdateCooldownProgressBar() {is=title K2Node_FunctionEntry}
   > n1 = Branch(Condition="Nearly Equal (Float)"(A=Get CooldownTimeRemaining(), B=0, ErrorTolerance=Get UpdateBarTick()))
   > Clear and Invalidate Timer by Handle(Handle=Get UpdateCooldownBarTimer())
   > UpdateReconReticle(isReconActive=true, Animate=true, UpdateCooldownBar=true)
   n2 = float - float(A=Get CooldownTimeRemaining(), B=Get UpdateBarTick())
   n3 = float / float(A=@n2, B=Get CooldownDuration() -> (CooldownDuration:float))
   SetScalarParameterValue(
    execute=@n1.else
    self=Get DynamicMaterial_CooldownBar()
    ParameterName=Progress
    Value=SelectFloat(A=@n3, B=float - float(A=1, B=@n3), bPickA=Get InvertProgressAnimation() -> (InvertProgressAnimation:bool))
   )
   > Set CooldownTimeRemaining(CooldownTimeRemaining=@n2)
  ]}
 UpdateReconReticle {nodes [
   n1 = UpdateReconReticle!() -> (then, isReconActive:bool=~, Animate:bool=~, UpdateCooldownBar:bool=~) {is=title K2Node_FunctionEntry}
   > n2 = Branch(Condition=@n1.Animate)
   > n3 = Get Scalar Parameter Value!(execute, self:Material Instance Dynamic Object Reference=Get Dynamic Material Reticle Shape(), ParameterName:Name=ReconReady) -> (then, ReturnValue:float) {is=title K2Node_CallFunction}
   n4 = Get isReconActive() -> (isReconActive:bool)
   > n5 = Branch(Condition="Nearly Equal (Float)"(A=@n3, B=SelectFloat(A=1, B=0, bPickA=@n4), ErrorTolerance=0))
   PlayAnimation(
    execute=@n5.else
    InAnimation=Get ReconActive() -> (ReconActive:Widget Animation Object Reference)
    PlayMode=Select(Option 0:EUMGSequencePlayMode Enum=Reverse, Option 1:EUMGSequencePlayMode Enum, Index=Get isReconActive()) -> (ReturnValue:EUMGSequencePlayMode Enum)
   )
   > Get!(execute:Exec) -> (then:Exec, else:Exec, UpdateCooldownBar:bool)
   > PlayAnimation(
    InAnimation=Get Anim_FadeIn() -> (Anim_FadeIn:Widget Animation Object Reference)
    PlayMode=Select(Option 1=Reverse, Index=Get isReconActive())
   )
   SetScalarParameterValue(
    execute=@n2.else
    self=Get Dynamic Material Reticle Shape()
    ParameterName=ReconReady
    Value=SelectFloat(A=1, B=0, bPickA=Get isReconActive())
   )
   > n6 = Get()
   > SetScalarParameterValue(self=Get DynamicMaterial_CooldownBar(), ParameterName=Opacity, Value=SelectFloat(A=0, B=1, bPickA=Get isReconActive()))
   SetScalarParameterValue(execute=@n6.else, self=Get DynamicMaterial_CooldownBar(), ParameterName=Opacity, Value=0)
  ]}
]
variables [
 As Fort Player Controller {type=FortPlayerController, default=null}
 CooldownDuration {type=number}
 CooldownTag {type=GameplayTag, default {tagName=Granted.Cooldown.StalkerKnife.StalkerMode}}
 CooldownTimeRemaining {type=number}
 Dynamic Material Reticle Shape {type=MaterialInstanceDynamic}
 DynamicMaterial_CooldownBar {type=MaterialInstanceDynamic}
 InvertProgressAnimation {type=boolean}
 Spectator Tag Remove Async Action {type=AbilityAsync_WaitGameplayTagRemoved, default=null}
 UpdateBarTick {type=number, default=0.1}
 UpdateCooldownBarTimer {type=TimerHandle}
 WaitGameplayTagAddAsyncAction {type=AbilityAsync_WaitGameplayTagAdded, default=null}
]
widgets {
 info {
  inheritedWidgetCount=1
  parentClass=/Script/FortniteUI.FortWeaponReticleExtensionWidgetBase
  rootWidgetClass=&Overlay
  widgetCount=5
 }
 widgets [
  Overlay_Reticle {widgetClassPath=&Overlay, widget=Overlay_Reticle}
  Reticle_Shape {
   widgetClassPath=&Image
   bIsVariable=true
   parent=Overlay_Reticle
   widget=Reticle_Shape
   properties {brush {
     bIsDynamicallyLoaded=false
     drawAs=Image
     imageType=FullColor
     mirroring=NoMirror
     resourceName=None
     resourceObject=/FNMStalkerKnife/UI/Materials/M_UI_StalkerKnife_Reticle
     tiling=NoTile
     imageSize=<100 100>
     margin=<ltrb 0 0 0 0>
     outlineSettings {
      bUseBrushTransparency=false
      roundingType=HalfHeightRadius
      width=0
      color {colorUseRule=UseColor_Specified, specifiedColor=<rgba 0 0 0 0>}
      cornerRadii=<0 0 0 0>
     }
     tintColor {colorUseRule=UseColor_Specified, specifiedColor=<rgba 1 1 1 1>}
     uVRegion {bIsValid=false, max=<0 0>, min=<0 0>}
    }}
   slot {class=OverlaySlot, properties {horizontalAlignment=HAlign_Center, verticalAlignment=VAlign_Center}}
  }
  Reticle_Dot {
   widgetClassPath=&Image
   parent=Overlay_Reticle
   widget=Reticle_Dot
   properties {brush {
     bIsDynamicallyLoaded=false
     drawAs=Image
     imageType=FullColor
     mirroring=NoMirror
     resourceName=None
     resourceObject=/FNMStalkerKnife/UI/Materials/MI_StalkerKnife_Reticle_Dot
     tiling=NoTile
     imageSize=<15 15>
     margin=<ltrb 0 0 0 0>
     outlineSettings {
      bUseBrushTransparency=false
      roundingType=HalfHeightRadius
      width=0
      color {colorUseRule=UseColor_Specified, specifiedColor=<rgba 0 0 0 0>}
      cornerRadii=<0 0 0 0>
     }
     tintColor {colorUseRule=UseColor_Specified, specifiedColor=<rgba 1 1 1 1>}
     uVRegion {bIsValid=false, max=<0 0>, min=<0 0>}
    }}
   slot {class=OverlaySlot, properties {horizontalAlignment=HAlign_Center, verticalAlignment=VAlign_Center}}
  }
  Reticle_ReconCooldown {
   widgetClassPath=&Image
   bIsVariable=true
   parent=Overlay_Reticle
   widget=Reticle_ReconCooldown
   properties {
    brush {
     bIsDynamicallyLoaded=false
     drawAs=Image
     imageType=FullColor
     mirroring=NoMirror
     resourceName=None
     resourceObject=/FNMStalkerKnife/UI/Materials/MI_UI_StalkerKnife_CooldownBar
     tiling=NoTile
     imageSize=<256 256>
     margin=<ltrb 0 0 0 0>
     outlineSettings {
      bUseBrushTransparency=false
      roundingType=HalfHeightRadius
      width=0
      color {colorUseRule=UseColor_Specified, specifiedColor=<rgba 0 0 0 0>}
      cornerRadii=<0 0 0 0>
     }
     tintColor {colorUseRule=UseColor_Specified, specifiedColor=<rgba 1 1 1 1>}
     uVRegion {bIsValid=false, max=<0 0>, min=<0 0>}
    }
    renderTransform {angle=0, scale=<1 1>, shear=<0 0>, translation=<65 20>}
   }
   slot {class=OverlaySlot, properties {horizontalAlignment=HAlign_Center, verticalAlignment=VAlign_Center}}
  }
  TutorialHighlight {widgetClassPath=&TutorialHighlight}
 ]
}
